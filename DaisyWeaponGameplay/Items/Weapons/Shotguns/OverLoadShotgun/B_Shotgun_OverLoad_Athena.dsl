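; Fortnite OverLoad shotgun weapon Blueprint that drives audio, visual and rumble feedback from magazine charge.
; Ammo changes update the AmmoCount material parameter and play indicator or fully-charged sounds, muted briefly after equip by a delayed flag.
; Firing picks full-magazine muzzle and rumble, plus first/third-person sounds using a hard-coded over-three-rounds threshold, with sound indicators.
; Charge State Check ends in an empty branch, and CurrentAmmo still prints debug text.
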(event Custom|EventChargeSounds
  (params (FullClip :type Boolean))
  (bind _weaponmesh (Variables|Visual|GetWeaponMesh))
  (if (Variables|Default|GetReloading)
    (if (== (Variables|Default|GetCurrentAmmo) (Variables|Default|GetMaxAmmo))
      (bind _returnvalue (Audio|SpawnSoundAttached :Sound "/DaisyWeaponGameplay/Items/Weapons/Shotguns/OverLoadShotgun/Sounds/OverLoad_FullyCharged_Cue.OverLoad_FullyCharged_Cue" :AttachToComponent _weaponmesh :AttachPointName "None" :LocationType "SnapToTarget" :bStopWhenAttachedToDestroyed false :VolumeMultiplier 1.0 :PitchMultiplier 1.0 :StartTime 0.0 :bAutoDestroy true))
      (Variables|Default|SetCurrentChargeSound _returnvalue)
      (else
        (Audio|SpawnSoundAttached :Sound "/DaisyWeaponGameplay/Items/Weapons/Shotguns/OverLoadShotgun/Sounds/OverLoad_IndicatorLight_Cue.OverLoad_IndicatorLight_Cue" :AttachToComponent _weaponmesh :AttachPointName "None" :LocationType "SnapToTarget" :bStopWhenAttachedToDestroyed false :VolumeMultiplier 1.0 :PitchMultiplier 1.0 :StartTime 0.0 :bAutoDestroy true)))))

(event FortWeapon|EventOnLocalUnEquip
  (Audio|Components|Audio|Stop (Variables|Default|GetCurrentChargeSound)))

(event Custom|Charge%20State%20Check
  (Utilities|FlowControl|Delay 0.2)
  (if (== (Variables|Default|GetCurrentAmmo) (Variables|Default|GetMaxAmmo))
    _))

(event EventOnPlayWeaponFireFX
  (params (bPersistentFire :type Boolean) (bSecondaryFire :type Boolean))
  (bind _weaponmesh (Variables|Visual|GetWeaponMesh))
  (|CallParentFunction bPersistentFire bSecondaryFire)
  (Variables|Default|SetReloading false)
  (Variables|Default|SetEquip false)
  (bind _ammo (CallFunction|CurrentAmmo))
  (Variables|Default|SetCurrentAmmo _ammo)
  (CallFunction|EventResetChargeSounds)
  (if (== (Variables|Default|GetCurrentAmmo) (Variables|Default|GetMaxAmmo))
    (Rendering|SetVisibility (Variables|Default|GetMaxMuzzle) true false)
    (Components|Activation|Activate (Variables|Default|GetMaxMuzzle) true)
    (ForceFeedback|SpawnForceFeedbackAttached "/DaisyWeaponGameplay/Items/Weapons/Shotguns/OverLoadShotgun/FF_Charge_Shotgun_MaxAmmo_Fire.FF_Charge_Shotgun_MaxAmmo_Fire" _weaponmesh)
    (else
      (Rendering|SetVisibility (Variables|Default|GetDefaultMuzzle) true false)
      (Components|Activation|Activate (Variables|Default|GetDefaultMuzzle) true)))
  (Rendering|Material|SetScalarParameterValueonMaterials _weaponmesh "AmmoCount" (Math|Conversions|ToFloat(Integer) (- (FortWeapon|GetMagazineAmmoCount) 1)))
  (CallFunction|ChargeStateCheck))

(event EventPlayWeaponFireCustomAudio
  (params (bPersistentFire :type Boolean) (bSecondaryFire :type Boolean))
  (|CallParentFunction bPersistentFire bSecondaryFire)
  (CallFunction|EventFiringSounds))

(event Custom|EventFiringSounds
  (bind _returnvalue (Transformation|GetActorLocation))
  (bind _currentammo (Variables|Default|GetCurrentAmmo))
  (bind _returnvalue_2 (> _currentammo 3))
  (bind _soundindicatorcomponent (Variables|SoundIndicator|GetSoundIndicatorComponent))
  (bind _weaponmesh (Variables|Visual|GetWeaponMesh))
  (Utilities|FlowControl|Sequence
    (:then_0
      (if (Pawn|IsLocallyControlled (Game|GetInstigator))
        (if _returnvalue_2
          ;[0.5573 0.5573 0.5573] First Person
          ;;[1 1 1] FullClip
          (bind _returnvalue_3 (Audio|SpawnSound2D "/DaisyWeaponGameplay/Items/Weapons/Shotguns/OverLoadShotgun/Sounds/OverLoad_FullFire_1P_Cue.OverLoad_FullFire_1P_Cue"))
          (bind _output_get (Variables|Default|SetPlayedSound (Variables|/Script/Engine.AudioComponent|GetSound _returnvalue_3)))
          (SoundIndicator|ShowSoundIndicator _soundindicatorcomponent _returnvalue _output_get 1.0 0)
          (else
            ;;[1 1 1] Normal
            (bind _returnvalue_4 (Audio|SpawnSound2D "/DaisyWeaponGameplay/Items/Weapons/Shotguns/OverLoadShotgun/Sounds/OverLoad_Fire_1P_Cue.OverLoad_Fire_1P_Cue"))
            (bind _output_get_2 (Variables|Default|SetPlayedSound (Variables|/Script/Engine.AudioComponent|GetSound _returnvalue_4)))
            (SoundIndicator|ShowSoundIndicator _soundindicatorcomponent _returnvalue _output_get_2 1.0 0)))
        (elif _returnvalue_2
          ;[0.401 0.401 0.401] Third Person
          ;;[1 1 1] FullClip
          (bind _returnvalue_5 (Audio|SpawnSoundAttached :Sound "/DaisyWeaponGameplay/Items/Weapons/Shotguns/OverLoadShotgun/Sounds/OverLoad_FullFire_3P_Cue.OverLoad_FullFire_3P_Cue" :AttachToComponent _weaponmesh :AttachPointName "None" :LocationType "SnapToTarget" :bStopWhenAttachedToDestroyed false :VolumeMultiplier 1.0 :PitchMultiplier 1.0 :StartTime 0.0 :bAutoDestroy true))
          (bind _output_get_3 (Variables|Default|SetPlayedSound (Variables|/Script/Engine.AudioComponent|GetSound _returnvalue_5)))
          (SoundIndicator|ShowSoundIndicator _soundindicatorcomponent _returnvalue _output_get_3 1.0 0)
          (Fortnite|ShowSoundIndicatorLocal :TrackedActor self :Location "0, 0, 0" :MaxAudibleDistance 0.0 :IndicatorType "Gunshot" :Tint "(R=0.000000,G=0.000000,B=0.000000,A=0.000000)" :SoundBase (Variables|Default|GetPlayedSound) :AttenuationMaxRangeMultiplier 1.0 :AttenuationIntensityMultiplier 1.0)
          (else
            ;;[1 1 1] Normal
            (bind _returnvalue_6 (Audio|SpawnSoundAttached :Sound "/DaisyWeaponGameplay/Items/Weapons/Shotguns/OverLoadShotgun/Sounds/OverLoad_Fire_Cue.OverLoad_Fire_Cue" :AttachToComponent _weaponmesh :AttachPointName "None" :LocationType "SnapToTarget" :bStopWhenAttachedToDestroyed false :VolumeMultiplier 1.0 :PitchMultiplier 1.0 :StartTime 0.0 :bAutoDestroy true))
            (bind _output_get_4 (Variables|Default|SetPlayedSound (Variables|/Script/Engine.AudioComponent|GetSound _returnvalue_6)))
            (SoundIndicator|ShowSoundIndicator _soundindicatorcomponent _returnvalue _output_get_4 1.0 0)))))
    (:then_1
      (CallFunction|ShowWeaponFireAudioVisual))))

(event EventOnAmmoCountChanged
  (bind _weaponmesh (Variables|Visual|GetWeaponMesh))
  (if (> (FortWeapon|GetMagazineAmmoCount) (Variables|Default|GetCurrentAmmo))
    (Variables|Default|SetReloading true)
    (else
      (Variables|Default|SetReloading false)))
  (CallFunction|MaxAmmo)
  (bind _ammo (CallFunction|CurrentAmmo))
  (Variables|Default|SetCurrentAmmo _ammo)
  (if (Variables|Default|GetEquip)
    (CallFunction|ChargeStateCheck)
    (elif (== (Variables|Default|GetCurrentAmmo) (Variables|Default|GetMaxAmmo))
      (bind _returnvalue (Audio|SpawnSoundAttached :Sound "/DaisyWeaponGameplay/Items/Weapons/Shotguns/OverLoadShotgun/Sounds/OverLoad_FullyCharged_Cue.OverLoad_FullyCharged_Cue" :AttachToComponent _weaponmesh :AttachPointName "None" :LocationType "SnapToTarget" :bStopWhenAttachedToDestroyed false :VolumeMultiplier 1.0 :PitchMultiplier 1.0 :StartTime 0.0 :bAutoDestroy true))
      (Variables|Default|SetCurrentChargeSound _returnvalue)
      (CallFunction|ChargeStateCheck)
      (else
        (Audio|SpawnSoundAttached :Sound "/DaisyWeaponGameplay/Items/Weapons/Shotguns/OverLoadShotgun/Sounds/OverLoad_IndicatorLight_Cue.OverLoad_IndicatorLight_Cue" :AttachToComponent _weaponmesh :AttachPointName "None" :LocationType "SnapToTarget" :bStopWhenAttachedToDestroyed false :VolumeMultiplier 1.0 :PitchMultiplier 1.0 :StartTime 0.0 :bAutoDestroy true)))))

(event FortWeapon|EventOnEquippedAndAttached
  (CallFunction|MaxAmmo)
  (CallFunction|CurrentAmmo)
  (Variables|Default|SetEquip true)
  (Utilities|FlowControl|Delay 0.4)
  (Variables|Default|SetEquip false)
  (CallFunction|ChargeStateCheck))

(event Custom|Reload%20Triggered%20Event
  (CallFunction|CurrentAmmo)
  (Utilities|FlowControl|Delay 0.2)
  (CallFunction|CurrentAmmo)
  (CallFunction|ChargeStateCheck))

(event EventMuzzlePlayReloadFX
  (params (Selection :type EFortReloadFXState%20Enum :default "ReloadStart"))
  (switch Utilities|FlowControl|Switch|SwitchonEFortReloadFXState Selection
    (:ReloadStart
      (|CallParentFunction "ReloadStart"))
    (:ReloadEnd
      (|CallParentFunction "ReloadEnd")))
  (if (not (Variables|Default|GetEquip))
    (Utilities|FlowControl|Delay 0.2)
    (if (== (FortWeapon|GetMagazineAmmoCount) (Variables|Default|GetMaxAmmo))
      (ForceFeedback|SpawnForceFeedbackAttached "/DaisyWeaponGameplay/Items/Weapons/Shotguns/TwoShotShotgun/FF_Shotgun_s21_Reload_MaxAmmo.FF_Shotgun_s21_Reload_MaxAmmo" (Variables|Visual|GetWeaponMesh)))))

(fn CurrentAmmo
  (bind _output_get (Variables|Default|SetCurrentAmmo (FortWeapon|GetLocalMagazineAmmoCount)))
  (Rendering|Material|SetScalarParameterValueonMaterials (Variables|Visual|GetWeaponMesh) "AmmoCount" (Math|Conversions|ToFloat(Integer) _output_get))
  (Development|PrintString (Utilities|String|ToString(Float) (Math|Conversions|ToFloat(Integer) _output_get)) true true "(R=1.000000,G=0.432372,B=0.000000,A=1.000000)")
  (return _output_get))

(fn GetActorAudioParameters
  (return 0))

(fn GetAffiliationsToShowFor
  (return (Utilities|Array|MakeArray "Hostile")))

(fn MaxAmmo
  (bind _output_get (Variables|Default|SetMaxAmmo (FortWeapon|GetBulletsPerClip)))
  (return _output_get))

(fn ShouldShowSoundIndicator
  (params (PlayerController :type Fort%20Player%20Controller%20Object%20Reference))
  (return true))

(fn ConstructionScript
  (|CallParentFunction))
