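# Niagara muzzle flash for the OverLoad shotgun, named as its MaxPower variant, built from one-shot bursts in six emitters.
# CircularBurst scales and erodes meshes, CoreFlash combines mesh, sprite, light and component renderers, and a smoke emitter adds drifting puffs.
# Three spark emitters add speed-stretched streaks: one GPU-simulated, another distance-culled and offset by User.OriginOffset.
# Time-of-day scalars adjust burst and GPU-spark brightness; User.Overheat is declared without visible use.
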NiagaraSystem /DaisyWeaponGameplay/Items/Weapons/Shotguns/OverLoadShotgun/FX/NS_Athena_Shotgun_MF_OverLoad_MaxPower
dependencies {
 usedDataInterfaces [
  &NiagaraDataInterfaceVectorCurve
  &NiagaraDataInterfaceCurve
  &NiagaraDataInterfaceVector2DCurve
  &NiagaraDataInterfaceColorCurve
 ]
 usedDynamicInputs []
 usedModules [&SystemState]
}
emitters {
 CircularBurst {
  emitterUpdateScript [
   &EmitterState()
   &SpawnBurst_Instantaneous(Spawn Count=2, Use Spawn Probability=false, UseLoopCountLimit=false)
  ]
  particleSpawnScript [
   &InitializeParticle(
    Color Mode=::NewEnumerator0
    Lifetime=0.17
    Mesh Scale Max=<2 2 11>
    Mesh Scale Min=<1 1 5>
    Mesh Scale Mode=::NewEnumerator2
    UsePositionOffset=false
    Write Material Random=true
   ) {module=InitializeParticle001}
   &InitialMeshOrientation(
    Mesh Axis To Orient=<0 0 1>
    Mesh Orientation Mode=::NewEnumerator2
    Orientation Coordinate Space=::Local
    Rotation=&RandomRangeVector(Minimum=<0 0 0>, Maximum=<0 0 1>)
    Use Rotation Vector=true
   ) {module=InitialMeshOrientation001}
  ]
  particleUpdateScript [
   /SetVariables(Particles.Scale=<0 1.875 0>) {enabled=false, isSetParameters=true}
   &ParticleState()
   &ScaleMeshSize(Scale Factor=&VectorFromFloat(Value=&FloatFromCurve(
      FloatCurve=DataInterface() {propertyValues {Curve {keys [
          {arriveTangent=4.8412, interpMode=RCIM_Cubic, time=0, value=0.1762}
          {arriveTangent=0.8097, interpMode=RCIM_Cubic, leaveTangent=0.8097, time=0.163, value=0.6122}
          {interpMode=RCIM_Cubic, leaveTangent=0.1287, time=1, value=0.9859}
         ]}}}
      Scale Curve=&RandomRangeFloat(Minimum=0.5, Maximum=1)
     )))
   /Game/Effects/Niagara/Modules/Materials/MaterialSmoke(
    Erode:NiagaraFloat=&FloatFromCurve(FloatCurve=DataInterface() {propertyValues {
       bOptimizeLUT=false
       bUseLUT=false
       Curve {keys [
         {arriveTangent=-0.0652, interpMode=RCIM_Cubic, leaveTangent=-0.0652, tangentMode=RCTM_User, time=0.11, value=0.97}
         {arriveTangent=-5.7797, interpMode=RCIM_Cubic, leaveTangent=-5.7797, tangentMode=RCTM_User, time=0.39, value=0.66}
         {arriveTangent=-0.342, interpMode=RCIM_Cubic, leaveTangent=-0.342, tangentMode=RCTM_User, time=0.9954, value=-0.74}
        ]}
      }})
    Shading:NiagaraFloat=&FloatFromCurve(FloatCurve=DataInterface() {propertyValues {
       bOptimizeLUT=false
       bUseLUT=false
       Curve {keys [
         <at 0.07 0.5 RCIM_Cubic>
         {arriveTangent=1.2205, interpMode=RCIM_Cubic, leaveTangent=1.2205, time=0.17, value=1.02}
         {interpMode=RCIM_Cubic, tangentMode=RCTM_User, time=0.3895, value=0.89}
         <at 0.8 0 RCIM_Cubic>
        ]}
      }})
   )
   &DynamicMaterialParameters(
    Index 0 Param 1=1
    Index 0 Param 3=1
    Index 0 Param 4=1
    Module.Param0WriteEnabled=true
    Module.Param1WriteEnabled=true
    Module.Param2WriteEnabled=true
    Module.Param3WriteEnabled=true
   )
   &Color(Color=&MakeLinearColorFromVectorAndFloat(
     "Vector (RGB)"=&VectorFromCurve(
      VectorCurve=DataInterface() {propertyValues {
        XCurve {keys [
          <at 0 0 RCIM_Cubic>
          {arriveTangent=-354.8556, interpMode=RCIM_Cubic, leaveTangent=-354.8556, tangentMode=RCTM_User, time=0.5, value=100}
          <at 1 0>
         ]}
        YCurve {keys [<at 0 1000 RCIM_Cubic>, {arriveTangent=-37.0931, interpMode=RCIM_Cubic, leaveTangent=-37.0931, time=0.5, value=18.5}, <at 1 0>]}
        ZCurve {keys [<at 0 2.6 RCIM_Cubic>, {arriveTangent=-2.6, interpMode=RCIM_Cubic, leaveTangent=-2.6, time=0.5, value=1.2}, <at 1 0>]}
       }}
      Scale Curve=&VectorFromFloat(Value=&FNTimeOfDayScalar(Night Scale=0.3, Morning Scale=0.3, Day Scale=0.5, Evening Scale=0.5))
     )
     "Float (Alpha)"=&FloatFromCurve(FloatCurve=DataInterface() {propertyValues {Curve {keys [<at 0 1 RCIM_Cubic>, {arriveTangent=-3.0318, interpMode=RCIM_Cubic, leaveTangent=-3.0318, tangentMode=RCTM_User, time=1, value=0}]}}})
    ))
  ]
  rendererClasses [&NiagaraMeshRendererProperties]
  renderers [{rendererClass=&NiagaraMeshRendererProperties, rendererIndex=0}]
 }
 CoreFlash {
  emitterUpdateScript [
   &EmitterState()
   &SpawnBurst_Instantaneous(Spawn Count=1, Use Spawn Probability=false)
  ]
  particleSpawnScript [
   &InitializeParticle(
    Color=<rgba 1 0.3636 0.01 1>
    Lifetime Max=0.2
    Lifetime Min=0.1
    Lifetime Mode=::NewEnumerator1
    Mesh Scale Max=<1.2 1.1 1.1>
    Mesh Scale Min=<0.9 0.9 0.9>
    Mesh Scale Mode=::NewEnumerator2
    Sprite Size=<100 100>
    Sprite Size Mode=::NewEnumerator4
    UsePositionOffset=false
    Write Material Random=true
   ) {module=InitializeParticle001}
   /Niagara/Modules/Spawn/Initialization/InitializeParticle(
    Color=<rgba 1 0.3636 0.01 1>
    Lifetime=&UniformRangedFloat(Minimum=0.1, Maximum=0.1)
    Mesh Scale=&UniformRangedVector(Minimum=<0.5 0.5 0.5>, Maximum=<0.7 0.6 0.6>)
    Module.Write Color=true
    Module.Write Lifetime=true
    Module.Write Mass=true
    Module.Write Material Random=true
    Module.Write Position=true
    Module.Write Scale=true
    Module.Write SpriteRotation=false
    Module.Write SpriteSize=true
    Position=&Add_Vector(A=/Niagara/DynamicInputs/Helpers/SimulationPosition(), B=<0 0 0>)
    Sprite Size=<100 100>
   ) {enabled=false}
   /Game/Effects/Niagara/Modules/Orientation/FN_Fan(Spin:NiagaraFloat=&UniformRangedFloat(Minimum=0, Maximum=1))
   /SetVariables(Particles.LightPositionNew:Vector3f=&Add_Vector(
     A=Linked() {linkedVariable {name=Particles.Position, type=NiagaraPosition}}
     B=/Niagara/DynamicInputs/Transforms/TransformPosition(Position=<25 0 0>, Source Space=::Local, Destination Space=::Simulation)
    )) {isSetParameters=true}
   &CameraOffset(Camera Offset Amount=6)
   /SetVariables(Particles.SpriteColor=&MultiplyLinearColor(
     Linear Color=$Particles.Color
     Scale Factor=&MakeFloat4FromVectorAndFloat("Vector (RGB)"=&VectorFromFloat(Value=1.5), "Float (Alpha)"=1)
    )) {isSetParameters=true}
  ]
  particleUpdateScript [
   &ParticleState()
   &ScaleColor(
    Scale RGB=&VectorFromFloat(Value=&FloatFromCurve(
      FloatCurve=DataInterface() {propertyValues {Curve {keys [{interpMode=RCIM_Cubic, tangentMode=RCTM_Break, time=0, value=1}, {interpMode=RCIM_Cubic, tangentMode=RCTM_Break, time=1, value=0}]}}}
      Scale Curve=5
     ))
    ScaleA=false
    ScaleRGB=true
   )
   &ScaleMeshSize(Scale Factor=&VectorFromFloat(Value=&FloatFromCurve(FloatCurve=DataInterface() {propertyValues {Curve {keys [{arriveTangent=2.2457, interpMode=RCIM_Cubic, leaveTangent=2.2457, tangentMode=RCTM_User, time=0, value=0}, <at 1 1>]}}})))
   &FN_Sprite_Attributes(
    Create SpriteSize Attribute=true
    SpriteSize=&Vector2DFromFloat(Value=50)
    Write to Module Namespace=true
   ) {module=FN_Sprite}
   &DynamicMaterialParameters(Write Parameter Index 0=false)
   &FN_Light_Attributes(
    Color=<rgba 1 0.4617 0.0819 1>
    Create Color Attribute=true
    Create Exponent Attribute=true
    Enable Particle Light Enable=true
    Exponent=4
    Light Enabled=true
    Scale Color=4
   )
   &FN_Light_Attributes(Color=<rgba 1 0.4617 0.0819 1>, Create Color Attribute=true, Enable Particle Light Enable=true, Light Enabled=true, Scale Color=0.6) {module=FN_Light_Attributes001}
  ]
  rendererClasses [
   &NiagaraLightRendererProperties
   &NiagaraMeshRendererProperties
   &NiagaraComponentRendererProperties
   &NiagaraSpriteRendererProperties
  ]
  renderers [
   {rendererClass=&NiagaraLightRendererProperties, rendererIndex=0}
   {rendererClass=&NiagaraMeshRendererProperties, rendererIndex=1}
   {rendererClass=&NiagaraComponentRendererProperties, rendererIndex=2}
   {rendererClass=&NiagaraSpriteRendererProperties, rendererIndex=3}
  ]
 }
 FN_Weapons_Smoke_BETA001 {
  emitterUpdateScript [
   &EmitterState()
   &SpawnBurst_Instantaneous(Spawn Count=&UniformRangedInt(Minimum=2, Maximum=5), Use Spawn Probability=false)
  ]
  particleSpawnScript [
   /Niagara/Modules/Spawn/Initialization/V2/InitializeParticle(
    Lifetime Max=1
    Lifetime Min=0.4
    Lifetime Mode=::NewEnumerator1
    Sprite Rotation Angle=&RandomRangeFloat(Minimum=0, Maximum=1)
    Sprite Rotation Mode=::NewEnumerator1
    Sprite Size Mode=::NewEnumerator1
    Uniform Sprite Size Max=100
    Uniform Sprite Size Min=60
    UsePositionOffset=false
    Write Material Random=true
   ) {module=InitializeParticle001}
   /SetVariables(Particles.InitSpriteRotRate=&UniformRangedFloat(Minimum=-3, Maximum=3)) {isSetParameters=true}
   &AddVelocityInCone(
    Cone Angle=10
    Module.Use Velocity Falloff On Cone Axis=false
    Velocity Strength=&UniformRangedFloat(Minimum=120, Maximum=400)
   )
   &InheritVelocity(Inherited Velocity Speed Limit=2000, Speed Limit=true, UseSpeedThreshold=false)
  ]
  particleUpdateScript [
   &ParticleState()
   &SpriteRotationRate(Rotation Rate=&Multiply_Float(
     A=$Particles.InitSpriteRotRate
     B=&FloatFromCurve(FloatCurve=DataInterface() {propertyValues {Curve {keys [<at -0.0046 1.0113 RCIM_Cubic>,<at 1.0011 0.09 RCIM_Cubic>]}}})
    ))
   &Drag(Use Linear Drag=true, Use Rotational Drag=false)
   &SolveForcesAndVelocity(Clamp Velocity=false, Limit Acceleration=false)
   &ScaleSpriteSize(Scale Factor=&Vector2DFromFloat(Value=&FloatFromCurve(
      FloatCurve=DataInterface() {propertyValues {Curve {keys [
          {arriveTangent=0.4478, interpMode=RCIM_Cubic, leaveTangent=0.4478, tangentMode=RCTM_User, time=0.0013, value=0.0367}
          {arriveTangent=0.9127, interpMode=RCIM_Cubic, leaveTangent=0.9127, tangentMode=RCTM_User, time=0.5367, value=0.8831}
          {arriveTangent=3.2289, interpMode=RCIM_Cubic, leaveTangent=3.2289, tangentMode=RCTM_User, time=1.0008, value=1.6042}
         ]}}}
      Scale Curve=1.25
     )))
   &MaterialSmoke(
    Erode:NiagaraFloat=&FloatFromCurve(FloatCurve=DataInterface() {propertyValues {
       bOptimizeLUT=false
       bUseLUT=false
       Curve {keys [
         {arriveTangent=-0.0652, interpMode=RCIM_Cubic, leaveTangent=-0.0652, tangentMode=RCTM_User, time=0.11, value=0.97}
         {arriveTangent=-5.7797, interpMode=RCIM_Cubic, leaveTangent=-5.7797, tangentMode=RCTM_User, time=0.39, value=0.66}
         {arriveTangent=-0.342, interpMode=RCIM_Cubic, leaveTangent=-0.342, tangentMode=RCTM_User, time=0.9954, value=-0.74}
        ]}
      }})
    Shading:NiagaraFloat=&FloatFromCurve(FloatCurve=DataInterface() {propertyValues {
       bOptimizeLUT=false
       bUseLUT=false
       Curve {keys [
         <at 0.07 0.5 RCIM_Cubic>
         {arriveTangent=1.2205, interpMode=RCIM_Cubic, leaveTangent=1.2205, time=0.17, value=1.02}
         {interpMode=RCIM_Cubic, tangentMode=RCTM_User, time=0.3895, value=0.89}
         <at 0.8 0 RCIM_Cubic>
        ]}
      }})
   )
   &CameraOffset(Camera Offset Amount=20)
   &DynamicMaterialParameters(
    Index 0 Param 1=&Add_Float(
     A=$Particles.RandomFloat
     B=&FloatFromCurve(FloatCurve=DataInterface() {propertyValues {Curve {keys [<at 0 1>,<at 1 0>]}}})
    )
    Index 0 Param 2=&FloatFromCurve(FloatCurve=DataInterface() {propertyValues {Curve {keys [
         {arriveTangent=0.8564, interpMode=RCIM_Cubic, leaveTangent=0.8564, tangentMode=RCTM_User, time=0.0046, value=-0.9753}
         {arriveTangent=4.2876, interpMode=RCIM_Cubic, leaveTangent=5.5268, tangentMode=RCTM_User, time=1, value=1}
        ]}}})
    Index 0 Param 3=&FloatFromCurve(FloatCurve=DataInterface() {propertyValues {Curve {keys [
         {arriveTangent=-22.7009, interpMode=RCIM_Cubic, leaveTangent=-9.0491, tangentMode=RCTM_Break, time=0, value=1}
         {arriveTangent=-0.541, interpMode=RCIM_Cubic, leaveTangent=-28.3175, tangentMode=RCTM_Break, time=0.3531, value=0}
        ]}}})
    Index 0 Param 4=&FloatFromCurve(FloatCurve=DataInterface() {propertyValues {Curve {keys [{arriveTangent=-3.2497, interpMode=RCIM_Cubic, leaveTangent=-3.2497, tangentMode=RCTM_User, time=0, value=1}, <at 1 0 RCIM_Cubic>]}}})
    Module.Param0WriteEnabled=true
    Module.Param1WriteEnabled=true
    Module.Param2WriteEnabled=true
    Module.Param3WriteEnabled=true
   )
   &ScaleColor(
    Scale Alpha=&FloatFromCurve(
     FloatCurve=DataInterface() {propertyValues {Curve {keys [<at -0.003 -0.006>,<at 0.1806 0.5257>,<at 0.4581 0.2965>,<at 1 0>]}}}
     Scale Curve=0.2
    )
    Scale RGB=/Niagara/DynamicInputs/ScaleByCurve/ScaleVectorByCurve(
     InitialValue:Vector3f=<0.8 0.7 0.6>
     ScaleCurve:NiagaraDataInterfaceVectorCurve=DataInterface() {propertyValues {XCurve {keys [<at 0 1>,<at 1 0>]}, YCurve {keys [<at 0 1>,<at 1 0>]}, ZCurve {keys [<at 0 1>,<at 1 0>]}}}
     CurveIndex:NiagaraFloat=$Particles.NormalizedAge
    )
    ScaleA=true
    ScaleRGB=true
   )
  ]
  rendererClasses [&NiagaraSpriteRendererProperties]
  renderers [
   {rendererClass=&NiagaraSpriteRendererProperties}
   {rendererClass=&NiagaraSpriteRendererProperties, rendererIndex=1}
  ]
 }
 FN_Weapons_Sparks_Beta {
  emitterUpdateScript [
   &EmitterState()
   &SpawnBurst_Instantaneous(Spawn Count=&UniformRangedInt(Minimum=5, Maximum=10), Use Spawn Probability=false)
  ]
  particleSpawnScript [
   &InitializeParticle(
    Color=<rgba 1 0.5 0.1 1>
    Lifetime Max=0.3
    Lifetime Min=0.05
    Lifetime Mode=::NewEnumerator1
    Sprite Size Max=<1.5 1.5>
    Sprite Size Min=<1 1>
    Sprite Size Mode=::NewEnumerator2
    UsePositionOffset=false
    Write Material Random=true
   ) {module=InitializeParticle001}
   &AddVelocityInCone(
    Cone Angle=35
    Module.Use Velocity Falloff On Cone Axis=false
    Velocity Strength=&UniformRangedFloat(Minimum=400, Maximum=600)
   )
   &SystemLocation(MaskSpawnGroup=false, Offset=<2 0 0>)
  ]
  particleUpdateScript [
   &ParticleState()
   &AccelerationForce(Acceleration=<0 0 -980>)
   &Drag(Use Linear Drag=true, Use Rotational Drag=false)
   &SolveForcesAndVelocity(Clamp Velocity=false, Limit Acceleration=false)
   &Color(
    Color=<rgba 1 0.442 0.0945 1>
    Scale Alpha=0.5
    Scale Color=&VectorFromCurve(
     VectorCurve=DataInterface() {propertyValues {
       XCurve {keys [<at -0.0011 1.0067>,<at 0.9989 0>]}
       YCurve {keys [<at -0.0011 1.0067>,<at 0.9989 0>]}
       ZCurve {keys [<at -0.0011 1.0067>,<at 0.9989 0>]}
      }}
     Scale Curve=<50 50 10>
    )
   )
   &ScaleSpriteSizeBySpeed(
    Max Scale Factor=<0.6 15>
    Min Scale Factor=<0.2 1>
    Sample Scale Factor By Curve=true
    Scale Factor Curve=DataInterface() {propertyValues {Curve {keys [
        <at -0.0077 0.0765 RCIM_Cubic>
        {arriveTangent=1.2463, interpMode=RCIM_Cubic, leaveTangent=1.2463, tangentMode=RCTM_User, time=0.9985, value=0.9549}
       ]}}}
    Velocity Threshold=600
   )
  ]
  rendererClasses [&NiagaraSpriteRendererProperties]
  renderers [{rendererClass=&NiagaraSpriteRendererProperties}]
 }
 SparksLong {
  emitterUpdateScript [
   &EmitterState(
    Enable Distance Culling=true
    Max Distance Response=::KillAfterParticlesFinish
    MaxDistance=3000
    Module.bUseMaxDistance=true
    Module.bUseMinDistance=false
    Scalability Mode=::NewEnumerator1
   )
   &SpawnBurst_Instantaneous(
    Spawn Count=&RandomRangeInt(Minimum=6, Maximum=24)
    Use Spawn Probability=false
    UseLoopCountLimit=false
   )
  ]
  particleSpawnScript [
   &InitializeParticle(
    Color Maximum=<rgba 3 1.571 0 1>
    Color Minimum=<rgba 15 0.4846 0 1>
    Color Mode=::NewEnumerator2
    Lifetime Max=1
    Lifetime Min=0.4
    Lifetime Mode=::NewEnumerator1
    Position Offset=$User.OriginOffset
    Sprite Size Max=<0.1 5>
    Sprite Size Min=<0.1 3>
    Sprite Size Mode=::NewEnumerator2
    UsePositionOffset=true
    Write Material Random=true
   ) {module=InitializeParticle001}
   &AddVelocityInCone(
    Cone Angle=80
    Module.Use Velocity Falloff On Cone Axis=false
    Velocity Distribution Along Cone Axis=1
    Velocity Strength=&UniformRangedFloat(Minimum=500, Maximum=3250)
   )
   &SystemLocation(MaskSpawnGroup=false)
  ]
  particleUpdateScript [
   &ParticleState()
   &AccelerationForce(Acceleration=<0 0 -980>)
   &Drag(Drag=4, Use Linear Drag=true, Use Rotational Drag=false)
   &SolveForcesAndVelocity(Clamp Velocity=false, Limit Acceleration=false)
   &ScaleSpriteSize(
    Scale Factor=&Vector2DFromCurve(
     Vector2Curve=DataInterface() {propertyValues {
       XCurve {keys [<at 0 1>,<at 1 0.2>]}
       YCurve {keys [{interpMode=RCIM_Cubic, leaveTangent=-1.3305, tangentMode=RCTM_Break, time=0, value=1}, <at 1 0>]}
      }}
     ScaleCurve=<8 3>
    )
    Scale Sprite Size Mode=::NewEnumerator2
   )
   &ScaleSpriteSizeBySpeed(
    Max Scale Factor=<1 8>
    Min Scale Factor=<0.5 0.5>
    Velocity Threshold:NiagaraFloat=&RandomRangeFloat(Minimum=500, Maximum=1500)
   )
  ]
  rendererClasses [&NiagaraSpriteRendererProperties]
  renderers [{rendererClass=&NiagaraSpriteRendererProperties}]
 }
 SparksLong001 {
  simTarget=GPUComputeSim
  emitterUpdateScript [
   &EmitterState()
   &SpawnBurst_Instantaneous(
    Spawn Count=&RandomRangeInt(Minimum=18, Maximum=28)
    Use Spawn Probability=false
    UseLoopCountLimit=false
   )
  ]
  particleSpawnScript [
   &InitializeParticle(
    Color=<rgba 55 16.3888 1 1>
    Lifetime Max=0.24
    Lifetime Min=0.14
    Lifetime Mode=::NewEnumerator1
    Sprite Size Max=<1 1.25>
    Sprite Size Min=<2 3>
    Sprite Size Mode=::NewEnumerator2
    UsePositionOffset=false
    Write Material Random=true
   ) {module=InitializeParticle001}
   &ShapeLocation(
    Hemisphere Distribution=<0.5 1>
    MaskSpawnGroup=false
    Non Uniform Scale=<2 1 1>
    Override Seed=false
    Rotation Mode=::NewEnumerator2
    Sphere Radius=5
    UseOwnerScale=true
    Yaw / Pitch / Roll=<-0.25 0 0>
   )
   &AddVelocity(Velocity=&Multiply_VectorByFloat(Vector=$Particles.ShapeLocation.ShapeVector, Float=12))
   &AddVelocity(Velocity=&RandomRangeVector(Minimum=<1800 0 0>, Maximum=<1000 0 0>)) {module=AddVelocity001}
  ]
  particleUpdateScript [
   &ParticleState()
   &Drag(Drag=2, Use Linear Drag=true, Use Rotational Drag=false)
   &ScaleSpriteSize(Uniform Curve Sprite Scale=DataInterface() {propertyValues {Curve {keys [<at 0 1 RCIM_Cubic>,<at 1 1 RCIM_Cubic>]}}})
   &ScaleColor(
    Scale Alpha=&FloatFromCurve(FloatCurve=DataInterface() {propertyValues {Curve {keys [<at 0 3 RCIM_Cubic>,<at 1 0 RCIM_Cubic>]}}})
    Scale RGB=&VectorFromCurve(
     VectorCurve=DataInterface() {propertyValues {
       XCurve {keys [<at 0.5 5 RCIM_Cubic>,<at 1 1 RCIM_Cubic>]}
       YCurve {keys [<at 0.5 2.5 RCIM_Cubic>,<at 1 1 RCIM_Cubic>]}
       ZCurve {keys [<at 0.5 0.8333 RCIM_Cubic>,<at 1 1 RCIM_Cubic>]}
      }}
     Scale Curve=&VectorFromFloat(Value=&FNTimeOfDayScalar(Night Scale=0.06, Morning Scale=0.01, Day Scale=0.15, Evening Scale=0.08))
    )
    ScaleA=true
    ScaleRGB=true
   )
   &SolveForcesAndVelocity(Clamp Velocity=false, Limit Acceleration=false)
   &ScaleSpriteSizeBySpeed(Max Scale Factor=<0.5 25.6>)
  ]
  rendererClasses [&NiagaraSpriteRendererProperties]
  renderers [{rendererClass=&NiagaraSpriteRendererProperties}]
 }
}
summary {
 systemName=NS_Athena_Shotgun_MF_OverLoad_MaxPower
 emitters [
  {emitterName=FN_Weapons_Sparks_Beta, rendererClasses [&NiagaraSpriteRendererProperties]}
  {emitterName=FN_Weapons_Smoke_BETA001, rendererClasses [&NiagaraSpriteRendererProperties]}
  {emitterName=SparksLong001, simTarget=GPUComputeSim, rendererClasses [&NiagaraSpriteRendererProperties]}
  {emitterName=CircularBurst, rendererClasses [&NiagaraMeshRendererProperties]}
  {emitterName=SparksLong, rendererClasses [&NiagaraSpriteRendererProperties]}
  {
   emitterName=CoreFlash
   rendererClasses [
    &NiagaraLightRendererProperties
    &NiagaraMeshRendererProperties
    &NiagaraComponentRendererProperties
    &NiagaraSpriteRendererProperties
   ]
  }
 ]
 userVariables [
  User.OriginOffset {type=Vector3f, description="", defaultValue {struct=Vector3f, value=<0 0 0>}}
  User.Overheat {type=NiagaraFloat, defaultValue=0, description=""}
 ]
}
