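# First-person fire sound cue for the OverLoad shotgun, built from sounds borrowed from other shotguns.
# A mixer layers a random PumpUp close fire shot, pitched up after a very short delay, with a quieter, later random TwoShot ammo impact.
# Modulators vary pitch on both layers and volume on the impact layer.
# The cue uses first-person weapon attenuation, sound class and gunfire concurrency settings.
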SoundCue /DaisyWeaponGameplay/Items/Weapons/Shotguns/OverLoadShotgun/Sounds/OverLoad_Fire_1P_Cue
root=Mixer_2
nodes [
 {id=Delay_0, children [Random_1], properties {delayMax=0.6, delayMin=0.6}}
 {id=Delay_1, children [Modulator_1], properties {delayMax=0.05, delayMin=0.05}}
 {id=Mixer_2, children [Delay_1,Modulator_0], properties {inputVolume [1,0.5]}}
 {id=Modulator_0, children [Delay_0], properties {pitchMax=1.1, pitchMin=0.9, volumeMax=1, volumeMin=0.8}}
 {id=Modulator_1, children [Random_2], properties {pitchMax=1.2, pitchMin=1.1}}
 {
  id=Random_1
  children [WavePlayer_10, WavePlayer_17, WavePlayer_18, WavePlayer_19, WavePlayer_20, WavePlayer_21]
  properties {weights [1,1,1,1,1,1]}
 }
 {id=Random_2, children [WavePlayer_3,WavePlayer_4,WavePlayer_5,WavePlayer_6], properties {weights [1,1,1,1]}}
 {id=WavePlayer_10, properties {soundWaveAssetPtr=/DaisyWeaponGameplay/Items/Weapons/Shotguns/TwoShotShotgun/Sounds/TwoShot_AmmoImpact_01}}
 {id=WavePlayer_17, properties {soundWaveAssetPtr=/DaisyWeaponGameplay/Items/Weapons/Shotguns/TwoShotShotgun/Sounds/TwoShot_AmmoImpact_02}}
 {id=WavePlayer_18, properties {soundWaveAssetPtr=/DaisyWeaponGameplay/Items/Weapons/Shotguns/TwoShotShotgun/Sounds/TwoShot_AmmoImpact_03}}
 {id=WavePlayer_19, properties {soundWaveAssetPtr=/DaisyWeaponGameplay/Items/Weapons/Shotguns/TwoShotShotgun/Sounds/TwoShot_AmmoImpact_04}}
 {id=WavePlayer_20, properties {soundWaveAssetPtr=/DaisyWeaponGameplay/Items/Weapons/Shotguns/TwoShotShotgun/Sounds/TwoShot_AmmoImpact_05}}
 {id=WavePlayer_21, properties {soundWaveAssetPtr=/DaisyWeaponGameplay/Items/Weapons/Shotguns/TwoShotShotgun/Sounds/TwoShot_AmmoImpact_06}}
 {id=WavePlayer_3, properties {soundWaveAssetPtr=&/PumpUp_Fire_Close_01}}
 {id=WavePlayer_4, properties {soundWaveAssetPtr=&/PumpUp_Fire_Close_02}}
 {id=WavePlayer_5, properties {soundWaveAssetPtr=&/PumpUp_Fire_Close_03}}
 {id=WavePlayer_6, properties {soundWaveAssetPtr=&/PumpUp_Fire_Close_04}}
]
properties {
 attenuationSettings=&Fort_Weapon_Projectile_FirstPerson_Athena_Att
 maxDistance=2097152
 pitchMultiplier=1.1
 soundClassObject=&Fort_WeaponRanged_Fire_1P_PSM
 volumeMultiplier=0.65
 attenuationOverrides {falloffDistance=5000, attenuationShapeExtents {x=1000}}
 concurrencyOverrides {bLimitToOwner=true, maxCount=1}
 concurrencySet [&Weapon_Gunfire_FirstPerson_Conc]
}
