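# Fire sound cue for the OverLoad Shotgun, built from PumpUp shotgun recordings in close, medium and distant layers.
# Each layer randomly picks one of three waves, varies pitch and volume, then applies its own projectile attenuation.
# Close routes to the third-person fire sound class, while medium and distant share the distant-fire class.
# A mixer blends the layers with close loudest by input level, under local and global gunfire concurrency sets.
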SoundCue /DaisyWeaponGameplay/Items/Weapons/Shotguns/OverLoadShotgun/Sounds/OverLoad_Fire_Cue
root=Mixer_3
nodes [
 {
  id=Attenuation_1
  children [Modulator_1]
  properties {attenuationSettings=&Fort_Weapon_Projectile_Medium_Mono_Athena_Att}
 }
 {id=Attenuation_2, children [Modulator_7], properties {attenuationSettings=&Fort_Weapon_Projectile_Close_Athena_Att}}
 {
  id=Attenuation_3
  children [Modulator_6]
  properties {attenuationSettings=&Fort_Weapon_Projectile_Distant_Mono_Athena_Att}
 }
 {id=Mixer_3, children [SoundClass_2,SoundClass_6,SoundClass_3], properties {inputVolume [1,0.65,0.55]}}
 {id=Modulator_1, children [Random_3], properties {pitchMax=1.15, pitchMin=1.05, volumeMax=1.8, volumeMin=1.6}}
 {id=Modulator_6, children [Random_4], properties {pitchMax=1, pitchMin=0.9, volumeMax=1.7, volumeMin=1.5}}
 {id=Modulator_7, children [Random_2], properties {pitchMax=1.15, pitchMin=1.05, volumeMin=1}}
 {id=Random_2, children [WavePlayer_0,WavePlayer_1,WavePlayer_2], properties {weights [1,1,1]}}
 {id=Random_3, children [WavePlayer_4,WavePlayer_5,WavePlayer_6], properties {weights [1,1,1]}}
 {id=Random_4, children [WavePlayer_8, WavePlayer_9, WavePlayer_19], properties {weights [1,1,1]}}
 {
  id=SoundClass_2
  children [Attenuation_2]
  properties {soundClassOverride=&Fort_WeaponRanged_Fire_3P_PSM}
 }
 {
  id=SoundClass_3
  children [Attenuation_3]
  properties {soundClassOverride=&Fort_WeaponRanged_Fire_Distant}
 }
 {
  id=SoundClass_6
  children [Attenuation_1]
  properties {soundClassOverride=&Fort_WeaponRanged_Fire_Distant}
 }
 {id=WavePlayer_0, properties {soundWaveAssetPtr=&/PumpUp_Fire_Close_01}}
 {id=WavePlayer_1, properties {soundWaveAssetPtr=&/PumpUp_Fire_Close_02}}
 {id=WavePlayer_19, properties {soundWaveAssetPtr=&/PumpUp_Fire_Distant_03}}
 {id=WavePlayer_2, properties {soundWaveAssetPtr=&/PumpUp_Fire_Close_03}}
 {id=WavePlayer_4, properties {soundWaveAssetPtr=&/PumpUp_Fire_Medium_01}}
 {id=WavePlayer_5, properties {soundWaveAssetPtr=&/PumpUp_Fire_Medium_02}}
 {id=WavePlayer_6, properties {soundWaveAssetPtr=&/PumpUp_Fire_Medium_03}}
 {id=WavePlayer_8, properties {soundWaveAssetPtr=&/PumpUp_Fire_Distant_01}}
 {id=WavePlayer_9, properties {soundWaveAssetPtr=&/PumpUp_Fire_Distant_02}}
]
properties {
 maxDistance=27500
 volumeMultiplier=0.6
 attenuationOverrides {falloffDistance=5000, attenuationShapeExtents {x=1000}}
 concurrencyOverrides {bLimitToOwner=true, maxCount=3}
 concurrencySet [&Weapon_Gunfire_Local_Conc, &Weapon_Gunfire_Global_Conc]
}
