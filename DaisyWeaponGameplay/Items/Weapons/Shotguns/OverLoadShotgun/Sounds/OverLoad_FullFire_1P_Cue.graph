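# First-person sound cue for the OverLoad shotgun's full fire, limited to one instance per owner.
# A quality switch shares two mixes across four levels, each layering randomized Overload and PumpUp fire waves with a quieter, delayed TwoShot ammo impact.
# The first two levels delay the Overload layer slightly and draw from six impact variants; the last two play it immediately with three.
# Modulators randomize pitch or volume on several layers.
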SoundCue /DaisyWeaponGameplay/Items/Weapons/Shotguns/OverLoadShotgun/Sounds/OverLoad_FullFire_1P_Cue
root=QualityLevel_0
nodes [
 {id=Delay_0, children [Random_1], properties {delayMax=0.6, delayMin=0.6}}
 {id=Delay_1, children [Random_4], properties {delayMax=0.6, delayMin=0.6}}
 {id=Delay_2, children [Modulator_2], properties {delayMax=0.1, delayMin=0.1}}
 {id=Mixer_0, children [Delay_2,Modulator_1], properties {inputVolume [0.9,1]}}
 {id=Mixer_1, children [Mixer_3,Modulator_4], properties {inputVolume [1,0.5]}}
 {id=Mixer_2, children [Mixer_0,Modulator_0], properties {inputVolume [1,0.5]}}
 {id=Mixer_3, children [Modulator_5,Modulator_3], properties {inputVolume [1,1]}}
 {id=Modulator_0, children [Delay_0], properties {pitchMax=1.1, pitchMin=0.9, volumeMax=1.1, volumeMin=0.9}}
 {id=Modulator_1, children [Random_7], properties {pitchMax=1.1, pitchMin=1.1, volumeMax=1.35, volumeMin=1.25}}
 {id=Modulator_2, children [Random_3], properties {pitchMax=1, pitchMin=1}}
 {id=Modulator_3, children [Random_0], properties {pitchMax=0.95, pitchMin=0.85, volumeMax=1.35, volumeMin=1.25}}
 {id=Modulator_4, children [Delay_1], properties {pitchMax=1.1, pitchMin=0.9, volumeMax=1.1, volumeMin=0.9}}
 {id=Modulator_5, children [Random_6], properties {pitchMax=1, pitchMin=1, volumeMax=1, volumeMin=0.8}}
 {id=QualityLevel_0, children [Mixer_2,Mixer_2,Mixer_1,Mixer_1]}
 {id=Random_0, children [WavePlayer_9, WavePlayer_11], properties {weights [1,1]}}
 {
  id=Random_1
  children [WavePlayer_10, WavePlayer_17, WavePlayer_18, WavePlayer_19, WavePlayer_20, WavePlayer_21]
  properties {weights [1,1,1,1,1,1]}
 }
 {id=Random_3, children [WavePlayer_2,WavePlayer_7], properties {weights [1,1]}}
 {id=Random_4, children [WavePlayer_14, WavePlayer_15, WavePlayer_16], properties {weights [1,1,1]}}
 {id=Random_6, children [WavePlayer_25, WavePlayer_26], properties {weights [1,1]}}
 {id=Random_7, children [WavePlayer_31, WavePlayer_32], properties {weights [1,1]}}
 {id=WavePlayer_10, properties {soundWaveAssetPtr=/DaisyWeaponGameplay/Items/Weapons/Shotguns/TwoShotShotgun/Sounds/TwoShot_AmmoImpact_01}}
 {id=WavePlayer_11, properties {soundWaveAssetPtr=&/PumpUp_Fire_Close_02}}
 {id=WavePlayer_14, properties {soundWaveAssetPtr=&TwoShot_AmmoImpact_01}}
 {id=WavePlayer_15, properties {soundWaveAssetPtr=/DaisyWeaponGameplay/Items/Weapons/Shotguns/TwoShotShotgun/Sounds/TwoShot_AmmoImpact_02}}
 {id=WavePlayer_16, properties {soundWaveAssetPtr=/DaisyWeaponGameplay/Items/Weapons/Shotguns/TwoShotShotgun/Sounds/TwoShot_AmmoImpact_03}}
 {id=WavePlayer_17, properties {soundWaveAssetPtr=&TwoShot_AmmoImpact_02}}
 {id=WavePlayer_18, properties {soundWaveAssetPtr=&TwoShot_AmmoImpact_03}}
 {id=WavePlayer_19, properties {soundWaveAssetPtr=/DaisyWeaponGameplay/Items/Weapons/Shotguns/TwoShotShotgun/Sounds/TwoShot_AmmoImpact_04}}
 {
  id=WavePlayer_2
  properties {soundWaveAssetPtr=&/OverloadFullTemp_Fire_Close_02}
 }
 {id=WavePlayer_20, properties {soundWaveAssetPtr=/DaisyWeaponGameplay/Items/Weapons/Shotguns/TwoShotShotgun/Sounds/TwoShot_AmmoImpact_05}}
 {id=WavePlayer_21, properties {soundWaveAssetPtr=/DaisyWeaponGameplay/Items/Weapons/Shotguns/TwoShotShotgun/Sounds/TwoShot_AmmoImpact_06}}
 {
  id=WavePlayer_25
  properties {soundWaveAssetPtr=&/OverloadFullTemp_Fire_Close_01}
 }
 {
  id=WavePlayer_26
  properties {soundWaveAssetPtr=&/OverloadFullTemp_Fire_Close_2}
 }
 {
  id=WavePlayer_31
  properties {soundWaveAssetPtr=&/PumpUp_Fire_Full_Close_02}
 }
 {
  id=WavePlayer_32
  properties {soundWaveAssetPtr=&/PumpUp_Fire_Full_Close_03}
 }
 {
  id=WavePlayer_7
  properties {soundWaveAssetPtr=&/OverloadFullTemp_Fire_Close_03}
 }
 {id=WavePlayer_9, properties {soundWaveAssetPtr=&/PumpUp_Fire_Close_01}}
]
properties {
 attenuationSettings=&Fort_Weapon_Projectile_FirstPerson_Athena_Att
 maxDistance=2097152
 soundClassObject=&Fort_WeaponRanged_Fire_1P_PSM
 volumeMultiplier=0.95
 attenuationOverrides {falloffDistance=5000, attenuationShapeExtents {x=1000}}
 concurrencyOverrides {bLimitToOwner=true, maxCount=1}
 concurrencySet [&Weapon_Gunfire_FirstPerson_Conc]
}
