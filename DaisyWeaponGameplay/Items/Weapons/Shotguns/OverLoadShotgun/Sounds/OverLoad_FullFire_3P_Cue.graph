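# Third-person sound cue for the OverLoad shotgun's full fire, built from PumpUp full-fire wave assets.
# A mixer blends close, medium and distant layers, each picking a random wave, varying volume, shifting pitch and applying its own distance attenuation.
# The close layer uses the third-person ranged fire sound class, while medium and distant share the distant fire class.
# Local and global gunfire concurrency sets are assigned to limit overlapping shots.
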SoundCue /DaisyWeaponGameplay/Items/Weapons/Shotguns/OverLoadShotgun/Sounds/OverLoad_FullFire_3P_Cue
root=Mixer_3
nodes [
 {
  id=Attenuation_1
  children [Modulator_1]
  properties {attenuationSettings=&Fort_Weapon_Projectile_Medium_Mono_Athena_Att}
 }
 {id=Attenuation_2, children [Modulator_0], properties {attenuationSettings=&Fort_Weapon_Projectile_Close_Athena_Att}}
 {
  id=Attenuation_3
  children [Modulator_6]
  properties {attenuationSettings=&Fort_Weapon_Projectile_Distant_Mono_Athena_Att}
 }
 {id=Mixer_3, children [SoundClass_2,SoundClass_6,SoundClass_3], properties {inputVolume [1,0.75,0.65]}}
 {id=Modulator_0, children [Random_7], properties {pitchMax=1.1, pitchMin=1.1, volumeMax=1.35, volumeMin=1.25}}
 {id=Modulator_1, children [Random_6], properties {pitchMax=0.95, pitchMin=0.85, volumeMax=1.8, volumeMin=1.6}}
 {id=Modulator_6, children [Random_8], properties {pitchMax=1, pitchMin=0.9, volumeMax=1.7, volumeMin=1.5}}
 {id=Random_6, children [WavePlayer_27, WavePlayer_28, WavePlayer_33], properties {weights [1,1,1]}}
 {id=Random_7, children [WavePlayer_31, WavePlayer_32], properties {weights [1,1]}}
 {id=Random_8, children [WavePlayer_34, WavePlayer_35, WavePlayer_36], properties {weights [1,1,1]}}
 {
  id=SoundClass_2
  children [Attenuation_2]
  properties {soundClassOverride=&Fort_WeaponRanged_Fire_3P_PSM}
 }
 {
  id=SoundClass_3
  children [Attenuation_3]
  properties {soundClassOverride=&Fort_WeaponRanged_Fire_Distant}
 }
 {
  id=SoundClass_6
  children [Attenuation_1]
  properties {soundClassOverride=&Fort_WeaponRanged_Fire_Distant}
 }
 {
  id=WavePlayer_27
  properties {soundWaveAssetPtr=&/PumpUp_Fire_Full_Medium_01}
 }
 {
  id=WavePlayer_28
  properties {soundWaveAssetPtr=&/PumpUp_Fire_Full_Medium_02}
 }
 {
  id=WavePlayer_31
  properties {soundWaveAssetPtr=&/PumpUp_Fire_Full_Close_02}
 }
 {
  id=WavePlayer_32
  properties {soundWaveAssetPtr=&/PumpUp_Fire_Full_Close_03}
 }
 {
  id=WavePlayer_33
  properties {soundWaveAssetPtr=&/PumpUp_Fire_Full_Medium_03}
 }
 {
  id=WavePlayer_34
  properties {soundWaveAssetPtr=&/PumpUp_Fire_Full_Distant_01}
 }
 {
  id=WavePlayer_35
  properties {soundWaveAssetPtr=&/PumpUp_Fire_Full_Distant_02}
 }
 {
  id=WavePlayer_36
  properties {soundWaveAssetPtr=&/PumpUp_Fire_Full_Distant_03}
 }
]
properties {
 maxDistance=27500
 volumeMultiplier=0.9
 attenuationOverrides {falloffDistance=5000, attenuationShapeExtents {x=1000}}
 concurrencyOverrides {bLimitToOwner=true, maxCount=3}
 concurrencySet [&Weapon_Gunfire_Local_Conc, &Weapon_Gunfire_Global_Conc]
}
