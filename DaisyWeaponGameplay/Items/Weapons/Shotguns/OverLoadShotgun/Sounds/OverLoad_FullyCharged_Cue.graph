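# Sound cue for the OverLoad shotgun's fully charged sound, played as weapon foley.
# A root mixer layers one of three random charged waves, a louder sub-mix of two indicator-light waves, and a quiet, borrowed Neon rocket charge-start wave.
# Fixed-value modulators pitch up the charged and rocket layers without random variation.
# It uses the OverLoad charging attenuation and caps concurrency at two per owner, stopping the oldest.
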SoundCue /DaisyWeaponGameplay/Items/Weapons/Shotguns/OverLoadShotgun/Sounds/OverLoad_FullyCharged_Cue
root=Mixer_0
nodes [
 {id=Mixer_0, children [Modulator_2,Mixer_2,Modulator_14], properties {inputVolume [0.8,1.4,0.2]}}
 {id=Mixer_2, children [WavePlayer_7,Modulator_4], properties {inputVolume [1,0.4]}}
 {id=Modulator_14, children [WavePlayer_17], properties {pitchMax=1.3, pitchMin=1.3, volumeMax=0.5, volumeMin=0.5}}
 {id=Modulator_2, children [Random_0], properties {pitchMax=1.3, pitchMin=1.3}}
 {id=Modulator_4, children [WavePlayer_8], properties {pitchMax=1, pitchMin=1}}
 {id=Random_0, children [WavePlayer_0,WavePlayer_1,WavePlayer_2], properties {weights [1,1,1]}}
 {id=WavePlayer_0, properties {soundWaveAssetPtr=&/OverLoad_FullyCharged_01}}
 {id=WavePlayer_1, properties {soundWaveAssetPtr=&/OverLoad_FullyCharged_02}}
 {id=WavePlayer_17, properties {soundWaveAssetPtr=/FNGameplayCues/Sounds/Fort_Weapons/Launchers/Neon/Neon_Rocket_ChargeStart_B04}}
 {id=WavePlayer_2, properties {soundWaveAssetPtr=&/OverLoad_FullyCharged_03}}
 {
  id=WavePlayer_7
  properties {soundWaveAssetPtr=&/OverLoad_IndicatorLight_01}
 }
 {
  id=WavePlayer_8
  properties {soundWaveAssetPtr=&/OverLoad_IndicatorLight_04}
 }
]
properties {
 attenuationSettings=&/Overload_Charging_Att
 bOverrideConcurrency=true
 maxDistance=150
 soundClassObject=&Fort_WeaponFoley
 volumeMultiplier=0.501
 attenuationOverrides {falloffDistance=1600}
 concurrencyOverrides {bLimitToOwner=true, maxCount=2, resolutionRule=StopOldest}
}
