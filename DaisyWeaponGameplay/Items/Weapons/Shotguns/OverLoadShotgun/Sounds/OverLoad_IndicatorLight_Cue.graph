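# Sound cue for the OverLoad Shotgun's indicator light, played as a weapon foley sound with the shotgun's charging attenuation asset.
# A mixer layers two indicator light waves, the first played directly and the second through a modulator with pitch variation flattened.
# The mixer sets the modulated second input to zero volume, so only the first wave is audible.
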SoundCue /DaisyWeaponGameplay/Items/Weapons/Shotguns/OverLoadShotgun/Sounds/OverLoad_IndicatorLight_Cue
root=Mixer_0
nodes [
 {id=Mixer_0, children [WavePlayer_0,Modulator_1], properties {inputVolume [1,0]}}
 {id=Modulator_1, children [WavePlayer_4], properties {pitchMax=1, pitchMin=1}}
 {
  id=WavePlayer_0
  properties {soundWaveAssetPtr=&/OverLoad_IndicatorLight_01}
 }
 {
  id=WavePlayer_4
  properties {soundWaveAssetPtr=&/OverLoad_IndicatorLight_04}
 }
]
properties {
 attenuationSettings=&/Overload_Charging_Att
 maxDistance=150
 soundClassObject=&Fort_WeaponFoley
 volumeMultiplier=0.401
 attenuationOverrides {falloffDistance=1600}
 concurrencyOverrides {resolutionRule=PreventNew}
}
