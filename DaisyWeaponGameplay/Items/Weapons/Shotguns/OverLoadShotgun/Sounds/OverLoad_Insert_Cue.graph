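# Sound cue for the OverLoad shotgun's shell insert sound, reusing three TwoShot shotgun reload-insert waves.
# An equally weighted random pick feeds a modulator that lowers pitch and slightly varies volume, then an attenuation node with spatialization turned off.
# The cue plays at about half volume in the weapon foley class with shared reload attenuation.
# Its concurrency override blocks new instances at the limit.
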SoundCue /DaisyWeaponGameplay/Items/Weapons/Shotguns/OverLoadShotgun/Sounds/OverLoad_Insert_Cue
root=Attenuation_13
nodes [
 {id=Attenuation_13, children [Modulator_13], properties {attenuationOverrides {bSpatialize=false}}}
 {id=Modulator_13, children [Random_0], properties {pitchMax=0.9, pitchMin=0.8, volumeMax=1, volumeMin=0.9}}
 {id=Random_0, children [WavePlayer_0,WavePlayer_1,WavePlayer_2], properties {weights [1,1,1]}}
 {id=WavePlayer_0, properties {soundWaveAssetPtr=/DaisyWeaponGameplay/Items/Weapons/Shotguns/TwoShotShotgun/Sounds/TwoShot_ReloadInsert_01}}
 {id=WavePlayer_1, properties {soundWaveAssetPtr=/DaisyWeaponGameplay/Items/Weapons/Shotguns/TwoShotShotgun/Sounds/TwoShot_ReloadInsert_02}}
 {id=WavePlayer_2, properties {soundWaveAssetPtr=/DaisyWeaponGameplay/Items/Weapons/Shotguns/TwoShotShotgun/Sounds/TwoShot_ReloadInsert_03}}
]
properties {
 attenuationSettings=&Fort_Weapon_Reload_Att
 bOverrideConcurrency=true
 maxDistance=2097152
 soundClassObject=&Fort_WeaponFoley
 volumeMultiplier=0.501
 attenuationOverrides {falloffDistance=1600}
 concurrencyOverrides {resolutionRule=PreventNew}
}
