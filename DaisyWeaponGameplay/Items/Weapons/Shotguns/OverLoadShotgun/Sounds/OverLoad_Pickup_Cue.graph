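# Sound cue for picking up the OverLoad shotgun, built from reused PumpUp shotgun waves.
# A mixer layers two boosted branches: a random pickup wave, and a random reload-end wave pitched down, each passed through its own modulator.
# It uses the weapon pickup attenuation and loot-collect sound class, with a concurrency override that allows one instance and blocks new plays while it sounds.
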SoundCue /DaisyWeaponGameplay/Items/Weapons/Shotguns/OverLoadShotgun/Sounds/OverLoad_Pickup_Cue
root=Mixer_1
nodes [
 {id=Mixer_1, children [Modulator_0,Modulator_1], properties {inputVolume [2,1.6]}}
 {id=Modulator_0, children [Random_0]}
 {id=Modulator_1, children [Random_1], properties {pitchMax=0.9, pitchMin=0.8}}
 {id=Random_0, children [WavePlayer_0,WavePlayer_1,WavePlayer_3], properties {weights [1,1,1]}}
 {id=Random_1, children [WavePlayer_5,WavePlayer_6,WavePlayer_7], properties {weights [1,1,1]}}
 {id=WavePlayer_0, properties {soundWaveAssetPtr=&/PumpUp_Pickup_01}}
 {id=WavePlayer_1, properties {soundWaveAssetPtr=&/PumpUp_Pickup_02}}
 {id=WavePlayer_3, properties {soundWaveAssetPtr=&/PumpUp_Pickup_03}}
 {id=WavePlayer_5, properties {soundWaveAssetPtr=&/PumpUp_ReloadEnd_01}}
 {id=WavePlayer_6, properties {soundWaveAssetPtr=&/PumpUp_ReloadEnd_02}}
 {id=WavePlayer_7, properties {soundWaveAssetPtr=&/PumpUp_ReloadEnd_03}}
]
properties {
 attenuationSettings=&Fort_Weapon_Pickup_Att
 bOverrideConcurrency=true
 maxDistance=3500
 soundClassObject=&Fort_Loot_Collect
 volumeMultiplier=0.3692
 concurrencyOverrides {maxCount=1, resolutionRule=PreventNew}
}
