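# Sound cue for the end of the OverLoad shotgun's reload, played as weapon foley with the shared weapon-reload attenuation.
# A random node picks between two equally weighted reload-end waves borrowed from the TwoShot shotgun, then a modulator slightly varies their pitch.
# Its concurrency override limits simultaneous copies per owner, stopping the oldest first.
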SoundCue /DaisyWeaponGameplay/Items/Weapons/Shotguns/OverLoadShotgun/Sounds/OverLoad_ReloadEnd_Cue
root=Modulator_14
nodes [
 {id=Modulator_14, children [Random_1], properties {pitchMax=1.1, pitchMin=0.9}}
 {id=Random_1, children [WavePlayer_0,WavePlayer_4], properties {weights [1,1]}}
 {id=WavePlayer_0, properties {soundWaveAssetPtr=/DaisyWeaponGameplay/Items/Weapons/Shotguns/TwoShotShotgun/Sounds/TwoShot_ReloadEnd_01}}
 {id=WavePlayer_4, properties {soundWaveAssetPtr=/DaisyWeaponGameplay/Items/Weapons/Shotguns/TwoShotShotgun/Sounds/TwoShot_ReloadEnd_02}}
]
properties {
 attenuationSettings=&Fort_Weapon_Reload_Att
 bOverrideConcurrency=true
 maxDistance=2250
 soundClassObject=&Fort_WeaponFoley
 volumeMultiplier=0.326
 attenuationOverrides {falloffDistance=1600}
 concurrencyOverrides {bLimitToOwner=true, maxCount=2, resolutionRule=StopOldest}
}
