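# Sound cue for the start of the OverLoad shotgun's reload, classed as weapon foley.
# It picks one of three equally weighted PumpUp reload-start waves at random and plays it through a modulator that randomly raises the pitch slightly.
# Volume is reduced, spatial falloff uses the shared weapon reload attenuation, and a concurrency override prevents new instances once the limit is reached.
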SoundCue /DaisyWeaponGameplay/Items/Weapons/Shotguns/OverLoadShotgun/Sounds/OverLoad_ReloadStart_Cue
root=Modulator_0
nodes [
 {id=Modulator_0, children [Random_0], properties {pitchMax=1.2, pitchMin=1.1}}
 {id=Random_0, children [WavePlayer_1,WavePlayer_2,WavePlayer_3], properties {weights [1,1,1]}}
 {id=WavePlayer_1, properties {soundWaveAssetPtr=&/PumpUp_ReloadStart_01}}
 {id=WavePlayer_2, properties {soundWaveAssetPtr=&/PumpUp_ReloadStart_02}}
 {id=WavePlayer_3, properties {soundWaveAssetPtr=&/PumpUp_ReloadStart_03}}
]
properties {
 attenuationSettings=&Fort_Weapon_Reload_Att
 bOverrideConcurrency=true
 maxDistance=2250
 soundClassObject=&Fort_WeaponFoley
 volumeMultiplier=0.251
 attenuationOverrides {falloffDistance=1600}
 concurrencyOverrides {resolutionRule=PreventNew}
}
